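# M_Distance is an unlit translucent particle material in the grenade gadget assets, flagged for use on mesh particles.
# Particle color RGB feeds two separate function calls, MF_FX_EmissiveAdjustments and a day/night auto bloom control.
# The graph has no textures or parameters of its own and shows no connection from either function to a material output.
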Material /Game/Gadgets/Assets/Grenade/M_Distance
graph {
 properties {bUsedWithMeshParticles=true, blendMode=BLEND_Translucent, pixelDepthOffsetMode=PDOM_Legacy, shadingModel=MSM_Unlit, usedShadingModels=Unlit}
 nodes [
  n1 = ParticleColor()
  MaterialFunctionCall(Emissive=@n1.RGB) {materialFunction=&MF_FX_EmissiveAdjustments}
  MaterialFunctionCall!(DayBloom_Mult, In=@n1.RGB) {materialFunction=/Game/Packages/Fortnite_UniqueMaterials_1/MH_Materials/DayNight_Auto_Bloom_Control}
 ]
}
