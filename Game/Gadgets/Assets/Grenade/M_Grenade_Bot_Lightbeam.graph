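# Unlit translucent particle sprite material for the bot grenade's light beam, with a moving highlight sweep.
# Panning smoke noise and a soft cross-beam mask blend two colors, then distort and bend the sweep band.
# Dynamic parameters drive sweep position, radius and end mask, while SweepHardness stays unused.
# Particle alpha erodes the depth-faded shape, and particle red desaturates the emissive before a shared adjustment function.
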Material /Game/Gadgets/Assets/Grenade/M_Grenade_Bot_Lightbeam
graph {
 properties {
  bUsedWithParticleSprites=true
  blendMode=BLEND_Translucent
  pixelDepthOffsetMode=PDOM_Legacy
  shadingModel=MSM_Unlit
  usedShadingModels=Unlit
 }
 nodes [
  n1 = ComponentMask(None=TextureCoordinate()) {g=true}
  n2 = Constant() {r=1}
  n3 = DynamicParameter() {defaultValue=<rgba 0 0.1 0.05 2>, paramNames [SweepPosition, SweepRadius, SweepHardness, SecondaryMask]}
  n4 = Constant() {r=0.25}
  n5 = Constant() {r=0.5}
  n6 = SphereMask(A=ComponentMask(None=TextureCoordinate()) {r=true}, B=@n5, Radius=@n5, Hardness=Constant())
  n7 = Subtract(A=@n6, B=OneMinus(None=ParticleColor().A))
  DepthFade(
   Opacity=Multiply(A=SphereMask(A=@n1, B=@n2, Radius=@n3.SecondaryMask, Hardness=@n4), B=Multiply(A=@n1, B=@n7))
   FadeDistance=ScalarParameter() {defaultValue=30, parameterName=FadeDistance}
  )
  n8 = ScalarParameter() {defaultValue=4, parameterName=Exponent}
  n9 = TextureCoordinate()
  n10 = AppendVector(A=ScalarParameter() {defaultValue=2.5, parameterName=RepeatU}, B=ScalarParameter() {defaultValue=0.8, parameterName=RepeatV})
  n11 = ScalarParameter() {defaultValue=0.22, parameterName=PanSpeedU}
  n12 = ScalarParameter() {defaultValue=3.2, parameterName=PanSpeedV}
  n13 = TextureSample(UVs=Panner(Coordinate=Multiply(A=@n9, B=@n10), Speed=AppendVector(A=@n11, B=@n12)) {bFractionalPart=true}) {texture=&T_Inky_Smoke_Tile}
  n14 = ScalarParameter() {defaultValue=0.2, parameterName=NoiseScale}
  n15 = LinearInterpolate(
   A=VectorParameter() {parameterName=ColorA, defaultValue=<rgba 0.022 0.2086 1.4 0>}.RGB
   B=VectorParameter() {parameterName=ColorB, defaultValue=<rgba 0.7986 5.3712 20 0>}.RGB
   Alpha=Add(A=Power(Base=@n7, Exp=@n8) {constExponent=4}, B=Multiply(A=@n13.R, B=@n14))
  )
  n16 = VectorParameter() {parameterName=ColorSweep, defaultValue=<rgba 8.8834 14.2136 80 0>}
  n17 = DynamicParameter() {defaultValue=<rgba 0 0.1 0.05 2>, paramNames [SweepPosition, SweepRadius, SweepHardness, SecondaryMask]}
  n18 = Add(A=ConstantBiasScale(None=@n13.R) {bias=-0.5, scale=0.05}, B=@n1)
  n19 = Sine(None=@n6) {period=4}
  n20 = OneMinus(None=@n19)
  n21 = ScalarParameter() {defaultValue=0.85, parameterName=BendBias}
  n22 = Multiply(A=OneMinus(None=Add(A=@n17.SweepPosition, B=@n18)), B=Saturate(None=LinearInterpolate(A=@n20, Alpha=@n21)))
  n23 = DynamicParameter() {defaultValue=<rgba 0 0.1 0.05 2>, paramNames [SweepPosition, SweepRadius, SweepHardness, SecondaryMask]}
  n24 = Multiply(A=SphereMask(A=Power(Base=@n22) {constExponent=8}, B=@n23.SweepRadius, Radius=@n23.SweepRadius, Hardness=@n23.SweepRadius), B=@n19)
  MaterialFunctionCall(Emissive=Desaturation(None=LinearInterpolate(A=@n15, B=@n16.RGB, Alpha=@n24), Fraction=ParticleColor().R)) {materialFunction=&MF_FX_EmissiveAdjustments}
 ]
}
