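# Unlit translucent material for beam trails, apparently the trail of the bot grenade gadget.
# Particle color RGB feeds MF_FX_EmissiveAdjustments for the glow color.
# A sphere mask on the V texture coordinate forms a soft, texture-free band that fades from the center toward both edges.
# The particle system tunes that falloff through the Exponent dynamic parameter, and particle alpha scales the result.
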Material /Game/Gadgets/Assets/Grenade/M_Grenade_Bot_Trail
graph {
 properties {bUsedWithBeamTrails=true, blendMode=BLEND_Translucent, pixelDepthOffsetMode=PDOM_Legacy, shadingModel=MSM_Unlit, usedShadingModels=Unlit}
 nodes [
  n1 = ParticleColor()
  MaterialFunctionCall(Emissive=@n1.RGB) {materialFunction=&MF_FX_EmissiveAdjustments}
  n2 = ComponentMask(None=TextureCoordinate()) {g=true}
  n3 = Constant() {r=0.5}
  n4 = Constant()
  Multiply(
   A=@n1.A
   B=Power(Base=SphereMask(A=@n2, B=@n3, Radius=@n3, Hardness=@n4), Exp=DynamicParameter() {paramNames [Exponent,Param2,Param3,Param4]}.Exponent)
  )
 ]
}
