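# Translucent smoke trail material for grenades, usable on Niagara ribbons and beam trails.
# Smoke textures sample swizzled UVs, optionally warped by a scrolling distortion normal map, and combine with particle color, a soft ramp, V-edge mask and fade.
# Time-of-day colors and an optional, desaturated sky atmosphere light term tint the emissive to suit scene lighting.
# Smoke density also drives subtle refraction, optionally scaled by a dynamic parameter.
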Material /Game/Gadgets/Assets/Grenade/M_Grenade_Trail
graph {
 properties {
  bEnableResponsiveAA=true
  bUsedWithBeamTrails=true
  bUsedWithNiagaraRibbons=true
  blendMode=BLEND_Translucent
  pixelDepthOffsetMode=PDOM_Legacy
  refractionMethod=RM_IndexOfRefraction
  translucencyLightingMode=TLM_VolumetricPerVertexNonDirectional
  usedShadingModels=Default Lit
 }
 nodes [
  n1 = TextureCoordinate() {coordinateIndex=1}
  n2 = MaterialFunctionCall(XY=@n1) {materialFunction=&Swizzle}
  n3 = Add(
   A=Multiply(A=@n1, B=ScalarParameter() {defaultValue=1.5, parameterName=Distortion UV Scale})
   B=Multiply(A=Time(), B=ScalarParameter() {defaultValue=1, parameterName=Distortion Speed})
  )
  n4 = Subtract(A=ComponentMask(None=TextureSample(UVs=@n3) {samplerType=SAMPLERTYPE_Normal, texture=/Game/Effects/Fort_Effects/Textures/T_DustDistortion}.RGB) {g=true, r=true}) {constB=0.5}
  n5 = ScalarParameter() {defaultValue=0.1, parameterName=Distortion Strength}
  n6 = StaticSwitchParameter(True=Add(A=@n2.YX, B=Multiply(A=@n4, B=@n5)), False=@n2.YX) {parameterName=Use_Distortion}
  n7 = TextureSample(UVs=@n6) {texture=/Game/Effects/Fort_Effects/Textures/Smoke/T_Limbo_Smoke_Fit}
  n8 = TextureCoordinate()
  n9 = Multiply(
   A=@n7.G
   B=TextureSample(UVs=MaterialFunctionCall(XY=@n8) {materialFunction=&Swizzle}.YX) {texture=/Game/Effects/Fort_Effects/Textures/Ramps/T_Soft_RampInOUT}.G
  )
  LinearInterpolate(A=Constant(), B=@n9, Alpha=@n7.G)
  n10 = Multiply(A=@n7.G, B=DynamicParameter().Param1)
  LinearInterpolate(
   A=Constant() {r=1}
   B=TextureSample(UVs=TextureCoordinate() {coordinateIndex=1}) {samplerType=SAMPLERTYPE_Normal, texture=&T_DustDistortion}.RGB
   Alpha=Multiply(A=StaticSwitchParameter(True=@n10, False=@n7.G) {parameterName=ControlRefAmount}, B=Constant() {r=0.07})
  )
  n11 = NamedRerouteUsage()
  n12 = ParticleColor()
  n13 = TextureSample(UVs=@n6) {texture=/Game/Effects/Fort_Effects/Textures/Smoke/T_Limbo_Smoke_Fit_Soft}
  n14 = Multiply(A=Multiply(A=@n12.RGB, B=LinearInterpolate(A=@n7.G, B=@n13.G, Alpha=@n7.G)), B=NamedRerouteUsage())
  MaterialFunctionCall(Emissive=StaticSwitchParameter(True=Add(A=@n11, B=@n14), False=@n14) {defaultValue=true, group=Shading, parameterName=Use SkyAtmosphereLightColor}) {materialFunction=&MF_FX_EmissiveAdjustments}
  Multiply(A=Multiply(A=Multiply(A=@n12.A, B=@n9), B=ScalarParameter() {defaultValue=1, parameterName=Fade}), B=NamedRerouteUsage())
  n15 = ComponentMask(None=TextureCoordinate()) {g=true}
  n16 = Subtract(A=Multiply(A=@n15, B=OneMinus(None=@n15))) {constB=0.05}
  NamedRerouteDeclaration(None=Saturate(None=Multiply(A=@n16) {constB=5})) {name=V Mask, nodeColor=<rgba 1 0.9609 0 1>}
  n17 = ScalarParameter() {defaultValue=0.075, desc=How much consider the sky atmosphere lighting, group=Shading, parameterName=SkyAtmosphere LightScale}
  n18 = WorldPosition()
  n19 = SkyAtmosphereLightIlluminance!(World Position=AppendVector(A=@n18.XY, B=Abs(None=@n18.Z)))
  n20 = Saturate(None=Add(A=Multiply(A=@n17, B=@n19), B=@n17) {constB=0.1})
  n21 = Constant()
  NamedRerouteDeclaration(None=Desaturation(
    None=ShadingPathSwitch(Default=@n20, Deferred=@n20, Mobile=@n21) {desc=Migrated from FeatureLevelSwitch}
    Fraction=ScalarParameter() {defaultValue=0.5, parameterName=SkyAtmos Desaturation, sliderMax=1}
   )) {name=SkyAtmos, nodeColor=<rgba 0 0.25 1 1>}
  NamedRerouteDeclaration(None=MaterialFunctionCall(
    Color_Morning=VectorParameter() {parameterName=Color Morning, defaultValue=<rgba 0.4844 0.439 0.439 1>}.RGB
    Color_Day=VectorParameter() {parameterName=Color Day, defaultValue=<rgba 0.3906 0.4046 0.4167 1>}.RGB
    Color_Evening=VectorParameter() {parameterName=Color Evening, defaultValue=<rgba 0.6667 0.6253 0.6042 1>}.RGB
    Color_Night=VectorParameter() {parameterName=Color Night, defaultValue=<rgba 0.5312 0.5633 0.6667 1>}.RGB
   ) {materialFunction=&Color_Time_of_Day}.Result) {name=TOD, nodeColor=<rgba 0.5469 1 0 1>}
 ]
}
