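# Translucent unlit material for the grenade trajectory target marker, with pulsing colored glow and scrolling scanlines.
# Emissive combines color, sine pulse, depth-fade intersection highlight and optional edge glow from UV1 or distance fields (UV1 only on mobile), compensated for eye adaptation.
# Opacity blends Fresnel with a pivot-based sphere mask and adds scanlines, UV or pseudo-screen-space.
# An optional UV mask path adds depth fades for softening and ground contact.
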Material /Game/Gadgets/Assets/Grenade/M_TrajectoryTarget
graph {
 properties {bUsedWithNanite=true, blendMode=BLEND_Translucent, pixelDepthOffsetMode=PDOM_Legacy, shadingModel=MSM_Unlit, usedShadingModels=Unlit}
 nodes [
  n1 = ScalarParameter() {
   defaultValue=1.68
   desc=Eye Adaptation in the old scene was 1.68 - scale to match old appearance
   group=FX Emissive EyeAdaptation
   parameterName=FX Emissive EyeAdaptation Prescale
   sliderMax=4
  }
  n2 = LinearInterpolate(
   B=DistanceToNearestSurface!(World Position)
   Alpha=ScalarParameter() {defaultValue=1, group=Edge Glow, parameterName=EdgeGlow_UseDistanceField}
  ) {constA=100000, constB=1000000}
  n3 = ComponentMask(None=TextureCoordinate() {coordinateIndex=1}) {r=true}
  n4 = ShadingPathSwitch(Default=DistanceFieldsRenderingSwitch!(No=@n3, Yes=Min(A=@n2, B=@n3)), Mobile=@n3) {desc=Migrated from FeatureLevelSwitch}
  n5 = Constant()
  n6 = ScalarParameter() {defaultValue=5, group=Edge Glow, parameterName=EdgeGlow_Radius}
  n7 = ScalarParameter() {defaultValue=1, group=Edge Glow, parameterName=EdgeGlow_Hardness}
  n8 = StaticSwitchParameter(
   True=Multiply(
    A=SphereMask(A=@n4, B=@n5, Radius=@n6, Hardness=@n7)
    B=ScalarParameter() {defaultValue=0.6, group=Edge Glow, parameterName=Edge Glow Intensity}
   )
   False=Constant()
  ) {
   desc="Doc: Open Parent material and copy link from there.\r\nhttps://confluence.it.epicgames.com/pages/viewpage.action?pageId=251430673"
   group=Edge Glow
   parameterName="UseEdgeGlow(ReadDescription)"
  }
  n9 = NamedRerouteDeclaration(None=@n8) {name=Edge Glow, nodeColor=<rgba 0.125 0 1 1>}
  n10 = Multiply(A=Time()) {constB=1.8}
  n11 = Add(A=Multiply(A=Sine(None=@n10)) {constB=0.25})
  n12 = DepthFade(FadeDistance=Constant() {r=16})
  n13 = Multiply(
   A=Multiply(A=Add(A=@n9, B=@n11), B=VectorParameter() {parameterName=Color, defaultValue=<rgba 0.015 0.2625 1.5 0>}.RGB)
   B=Add(A=OneMinus(None=@n12))
  )
  EyeAdaptationInverse(
   LightValueInput=ComponentMask(None=Multiply(A=@n13, B=@n1) {constB=1.3}) {b=true, g=true, r=true}
   AlphaInput=ScalarParameter() {defaultValue=1, desc=Mix to Original, group=FX Emissive EyeAdaptation, parameterName=FX Emissive EyeAdaptation Alpha}
  )
  n14 = ScalarParameter() {defaultValue=2, parameterName=FresnelExponent}
  n15 = LinearInterpolate(
   A=VertexNormalWS()
   B=Transform(None=TextureSampleParameter2D() {parameterName=Normal, samplerType=SAMPLERTYPE_Normal, texture=&FlatNormalBlue}.RGB)
   Alpha=ScalarParameter() {parameterName=UseNormalTexture}
  )
  n16 = Abs(None=DotProduct(A=@n15, B=CameraVectorWS()))
  n17 = LinearInterpolate(
   A=MaterialFunctionCall(Exponent=@n14, Vertex / Pixel Normal=@n15) {materialFunction=&MF_Fresnel_FX}.Fresnel Inverted
   B=Power(Base=OneMinus(None=@n16), Exp=@n14)
   Alpha=ScalarParameter() {parameterName=UseFresnelforTwoSide}
  )
  n18 = NamedRerouteUsage()
  n19 = ScalarParameter() {group=SphereMaskOpacity, parameterName=ReplaceFresnelWithSphereMask, sortPriority=-10}
  n20 = PreSkinnedPosition()
  n21 = ComponentMask(None=Transform(None=VertexInterpolator(VS=@n20).PS) {transformSourceType=TRANSFORMSOURCE_Local, transformType=TRANSFORM_View}) {g=true}
  n22 = ScalarParameter() {defaultValue=3, parameterName=ScreenScapeScanlineTileHorizontal}
  n23 = Time() {bOverride_Period=true, period=0.2}
  n24 = Multiply(
   A=TextureCoordinate()
   B=MaterialFunctionCall(X=Constant() {r=1}, Y=ScalarParameter() {defaultValue=32, parameterName=ScanlineTileHorizontal}) {materialFunction=&MakeFloat2}.Result
  )
  n25 = StaticSwitchParameter(
   True=Subtract(A=Divide(A=@n21, B=@n22), B=Multiply(A=@n23) {constB=5})
   False=ComponentMask(None=Panner(Coordinate=@n24) {speedY=-5}) {g=true}
  ) {parameterName=PseudoScreenSpaceScanlines}
  n26 = Floor(None=Add(A=Frac(None=@n25)) {constB=0.5})
  n27 = Max(A=NamedRerouteUsage(), B=Add(A=LinearInterpolate(A=@n17, B=@n18, Alpha=@n19), B=Multiply(A=@n26) {constB=0.03}))
  n28 = TextureCoordinate()
  n29 = Multiply(
   A=@n27
   B=SphereMask(A=ComponentMask(None=@n28) {g=true}, B=Constant(), Radius=ScalarParameter() {defaultValue=1, parameterName=MaskSize}) {hardnessPercent=0}
  )
  n30 = ScalarParameter() {parameterName=DepthFade}
  n31 = OneMinus(None=DepthFade(FadeDistance=ScalarParameter() {parameterName=DepthFade_ground}))
  n32 = Constant() {r=0.7}
  LinearInterpolate(A=@n27, B=Add(A=DepthFade(Opacity=@n29, FadeDistance=@n30), B=Multiply(A=@n31, B=@n32)), Alpha=ScalarParameter() {parameterName=Mask})
  n33 = SphereMask(
   A=WorldPosition().XYZ
   B=TransformPosition(None=VectorParameter() {group=SphereMaskOpacity, parameterName=SphereMaskPivot, defaultValue=<rgba 0 0 0 1>}.RGB) {transformType=TRANSFORMPOSSOURCE_World}
   Radius=ScalarParameter() {defaultValue=100, group=SphereMaskOpacity, parameterName=SphereMaskFalloff}
   Hardness=ScalarParameter() {group=SphereMaskOpacity, parameterName=SphereMaskHardness}
  )
  n34 = OneMinus(None=@n33)
  n35 = ScalarParameter() {group=SphereMaskOpacity, parameterName=InvertSphereMask}
  NamedRerouteDeclaration(None=Power(
    Base=LinearInterpolate(A=@n34, B=@n33, Alpha=@n35)
    Exp=ScalarParameter() {defaultValue=1, group=SphereMaskOpacity, parameterName=SphereMaskExponent}
   )) {name=SphereMaskOpacity, nodeColor=<rgba 1 0.1172 0 1>}
 ]
}
