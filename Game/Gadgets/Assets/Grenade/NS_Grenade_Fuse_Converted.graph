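# Grenade fuse Niagara effect, apparently converted from Cascade, with sparks, glow light and trail.
# Fuse001 sprays speed-stretched sparks in a cone driven by smoothed owner orientation, disturbed by gravity, wind, drag and curl noise.
# GPU sparks arc under strong downward acceleration, while a tiny sprite carries a fading light.
# A distance-spawned ribbon rises, widens and cools from orange to dark grey; the `fuse` glow emitter is disabled.
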NiagaraSystem /Game/Gadgets/Assets/Grenade/NS_Grenade_Fuse_Converted
dependencies {
 usedDataInterfaces [&NiagaraDataInterfaceCurve, &NiagaraDataInterfaceVectorCurve]
 usedDynamicInputs []
 usedModules [&SystemState]
}
emitters {
 Fuse001 {
  emitterSpawnScript [/SetVariables(
    Emitter.CurrOri:Quat4f=&FindQuatBetweenTwoVectors(From Vector=<0 0 1>, To Vector=$Engine.Owner.SystemZAxis)
    Emitter.PrevOri:Quat4f=&FindQuatBetweenTwoVectors(From Vector=<0 0 1>, To Vector=$Engine.Owner.SystemZAxis)
   ) {enabled=false, isSetParameters=true}]
  emitterUpdateScript [
   &EmitterState(Life Cycle Mode=::NewEnumerator1, Loop Duration=2, UseLoopDelay=false)
   /SetVariables(Emitter.CurrOri:Quat4f=&FindQuatBetweenTwoVectors(From Vector=<0 0 1>, To Vector=$Engine.Owner.SystemZAxis)) {isSetParameters=true}
   &SpawnRate(SpawnRate=55, Use Spawn Probability=false)
   &InterpolateOverTime(Rate Of Change=15, Target Quat Value:Quat4f=$Emitter.CurrOri, Variable Type=::NewEnumerator5)
   /SetVariables(Emitter.PrevOri:Quat4f=$Emitter.CurrOri) {enabled=false, isSetParameters=true}
  ]
  particleSpawnScript [
   &InitializeParticle(
    Lifetime Max=0.4
    Lifetime Min=0.2
    Lifetime Mode=::NewEnumerator1
    Mass Mode=::NewEnumerator3
    Position Offset=<0 0 12>
    Position Offset Coordinate Space=::Local
    Sprite Rotation Mode=::NewEnumerator2
    Sprite Size=<4 8>
    Sprite Size Mode=::NewEnumerator4
    UsePositionOffset=true
    Write Material Random=true
   )
   &ShapeLocation(MaskSpawnGroup=false, Override Seed=false, Sphere Radius=1, UseOwnerScale=true)
   &AddVelocity(
    Cone Angle=35
    Cone Axis=&RotateVectorByQuaternion(
     Vector To Rotate=<0 0 1>
     Quaternion=/Niagara/DynamicInputs/Rotation/InvertQuaternion(Quat:Quat4f=$Emitter.InterpolateOverTime.InterpolateOverTime_Quaternion.Moving Average)
    )
    Distribution Along Cone Axis=1
    Inner Cone Angle=15
    Module.UseConeFalloff=true
    Module.UseVelDistribution=true
    Override Seed=false
    Rotation Coordinate Space=::Simulation
    Speed Falloff From Cone Axis=1
    Velocity Mode=::NewEnumerator2
    Velocity Speed=&RandomRangeFloat(Minimum=125, Maximum=444)
   )
  ]
  particleUpdateScript [
   &ParticleState()
   &GravityForce()
   &WindForce(Wind Speed=<35 0 0>, Wind Speed Scale=5)
   &Drag(Drag=&RandomRangeFloat(Minimum=0, Maximum=5), Use Linear Drag=true, Use Rotational Drag=false)
   &ScaleColor(
    Scale Alpha=&FloatFromCurve(FloatCurve=DataInterface() {propertyValues {Curve {keys [<at -0.0046 1.0113 RCIM_Cubic>,<at 1.0011 0.002 RCIM_Cubic>]}}})
    ScaleA=true
    ScaleRGB=true
   ) {enabled=false}
   &InheritVelocity(Speed Limit=false, UseSpeedThreshold=false) {enabled=false}
   &CurlNoiseForce(Bias Noise=false, Noise Frequency=44, Noise Strength=555, Pan Noise=false)
   &SolveForcesAndVelocity(Clamp Velocity=false, Limit Acceleration=false)
   &ScaleSpriteSizeBySpeed(Max Scale Factor=<1 3>, Min Scale Factor=<0.3 0.5>)
   &Color(Color=&MakeLinearColorFromVectorAndFloat(
     "Vector (RGB)"=<238.589 81.2308 24.9103>
     "Float (Alpha)"=&FloatFromCurve(FloatCurve=DataInterface() {propertyValues {Curve {keys [<at 0 0 RCIM_Cubic>,<at 0.0967 0.1386 RCIM_Cubic>,<at 1 0>]}}})
    ))
  ]
  rendererClasses [&NiagaraSpriteRendererProperties]
  renderers [{rendererClass=&NiagaraSpriteRendererProperties}]
 }
 GPU_up_Delay {
  simTarget=GPUComputeSim
  emitterUpdateScript [
   &EmitterState(Life Cycle Mode=::NewEnumerator1, Loop Duration=2, UseLoopDelay=true)
   &SpawnBurst_Instantaneous(Spawn Count=0, Use Spawn Probability=false, UseLoopCountLimit=false)
   &SpawnRate(SpawnRate=50, Use Spawn Probability=false)
  ]
  particleSpawnScript [
   &InitializeParticle(
    Lifetime=&RandomRangeFloat(Minimum=0.2, Maximum=0.5)
    Sprite Rotation Angle=&RandomRangeFloat(Minimum=0, Maximum=1)
    Sprite Rotation Mode=::NewEnumerator1
    Sprite Size=&Vector2DFromFloat(Value=&RandomRangeFloat(Minimum=1, Maximum=3))
    Sprite Size Mode=::NewEnumerator4
    UsePositionOffset=false
    Write Material Random=true
   )
   &AddVelocity(Velocity=&RandomRangeVector(Minimum=<30 30 30>, Maximum=<-30 -30 400>))
   &ShapeLocation(
    Cylinder Height=2
    Cylinder Radius=20
    MaskSpawnGroup=false
    Override Seed=false
    Shape Primitive=::NewEnumerator1
    Surface Only=false
    UseOwnerScale=true
   )
  ]
  particleUpdateScript [
   &ParticleState()
   &ScaleSpriteSize(
    Scale Sprite Size Mode=::NewEnumerator0
    Uniform Scale Factor=&FloatFromCurve(FloatCurve=DataInterface() {propertyValues {Curve {keys [
         <at -0.0044 0.2654>
         {arriveTangent=0.9116, interpMode=RCIM_Cubic, leaveTangent=0.9116, time=0.24, value=1.0117}
         {arriveTangent=-1.3312, interpMode=RCIM_Cubic, leaveTangent=-1.3312, time=0.8015, value=1}
         <at 1 0>
        ]}}})
   )
   &AccelerationForce(Acceleration=<0 0 -2200>)
   &SolveForcesAndVelocity(Clamp Velocity=false, Limit Acceleration=false)
   &Color(Color=&MakeLinearColorFromVectorAndFloat(
     "Vector (RGB)"=&VectorFromCurve(VectorCurve=DataInterface() {propertyValues {
        XCurve {keys [
          <at 0.0095 5107.2236 RCIM_Cubic>
          {arriveTangent=-6404.0493, interpMode=RCIM_Cubic, leaveTangent=-6404.0493, time=0.0697, value=5002.8364}
          <at 0.806 6.1037>
         ]}
        YCurve {keys [
          <at 0.0095 776.5526 RCIM_Cubic>
          {arriveTangent=-974.8029, interpMode=RCIM_Cubic, leaveTangent=-974.8029, time=0.0697, value=760.663}
          <at 0.806 0.0772>
         ]}
        ZCurve {keys [
          <at 0.0095 54.4699 RCIM_Cubic>
          {arriveTangent=-59.9813, interpMode=RCIM_Cubic, leaveTangent=-59.9813, time=0.0697, value=181.8841}
          <at 0.806 6.692>
         ]}
       }})
     "Float (Alpha)"=&ClampFloat(Float=&FloatFromCurve(FloatCurve=DataInterface() {propertyValues {Curve {keys [<at 0.8224 1>,<at 1 0>]}}}))
    ))
  ]
  rendererClasses [&NiagaraSpriteRendererProperties]
  renderers [{rendererClass=&NiagaraSpriteRendererProperties}]
 }
 Particle_Emitter {
  emitterUpdateScript [
   &EmitterState(Inactive Response=::NewEnumerator1, Life Cycle Mode=::NewEnumerator1, Loop Behavior=::NewEnumerator1, Loop Duration=1, UseLoopDelay=true)
   &SpawnBurst_Instantaneous(Use Spawn Probability=false, UseLoopCountLimit=false)
  ]
  particleSpawnScript [&InitializeParticle(Sprite Size=<0.01 0.01>, Sprite Size Mode=::NewEnumerator4, UsePositionOffset=false, Write Material Random=true)]
  particleUpdateScript [
   /SetVariables(Particles.LightExponent=5) {isSetParameters=true}
   &ParticleState(Kill Particles When Lifetime Has Elapsed=false, Let Infinitely Lived Particles Die When Emitter Deactivates=true)
   &Color(Color=&MakeLinearColorFromVectorAndFloat(
     "Vector (RGB)"=<3.2606 0.5208 0.2273>
     "Float (Alpha)"=&ClampFloat(Float=&FloatFromCurve(FloatCurve=DataInterface() {propertyValues {Curve {keys [<at 0 1>,<at 1 0>]}}}))
    ))
   &CascadeConversion_LightProperties(LightBrightnessScale=20, LightColorScale=<1 1 1>, LightRadiusScale=27000)
  ]
  rendererClasses [&NiagaraSpriteRendererProperties, &NiagaraLightRendererProperties]
  renderers [
   {rendererClass=&NiagaraSpriteRendererProperties}
   {rendererClass=&NiagaraLightRendererProperties, rendererIndex=1}
  ]
 }
 Ribbon {
  emitterSpawnScript [/SetVariables(Emitter.Static bool=true) {isSetParameters=true}]
  emitterUpdateScript [
   &EmitterState(Life Cycle Mode=::NewEnumerator1, Loop Duration=1, UseLoopDelay=true)
   &SpawnPerUnit(
    Module.Use Max Movement Threshold=true
    Module.Use Movement Tolerance=true
    Movement Tolerance=0.3
    Spawn Spacing:NiagaraFloat=&Divide_Float(A=1, B=&Divide_Float(A=1, B=80))
    Use Spawn Probability=false
   )
  ]
  particleSpawnScript [
   /SetVariables(Particles.Acceleration:Vector3f=<0 0 250>) {isSetParameters=true}
   &InitializeRibbon(
    Color=<rgba 0.7135 0.7135 0.7135 1>
    Ribbon Width=&Multiply_Float(A=15, B=2)
    UsePositionOffset=false
    Write Material Random=true
   )
   &DynamicMaterialParameters(Module.Index0Param0Write=true, Module.Index0Param1Write=false, Module.Index0Param2Write=false, Module.Index0Param3Write=false)
  ]
  particleUpdateScript [
   /SetVariables(Particles.ScaledRibbonWidth:NiagaraFloat=&Multiply_Float(
     A=$Particles.RibbonWidth
     B=&FloatFromCurve(FloatCurve=DataInterface() {propertyValues {Curve {keys [
          {arriveTangent=9.5068, interpMode=RCIM_Cubic, time=-0.0056, value=0.0871}
          {arriveTangent=0.0733, leaveTangent=0.0733, time=0.9928, value=6.8584}
         ]}}})
    )) {isSetParameters=true}
   &ParticleState()
   &Color(Color=&MakeLinearColorFromVectorAndFloat(
     "Vector (RGB)"=&VectorFromCurve(VectorCurve=DataInterface() {propertyValues {
        XCurve {keys [
          <at 0.1712 6.2357>
          {arriveTangent=-19.1968, interpMode=RCIM_Cubic, leaveTangent=-19.1968, time=0.3943, value=0.5958}
          {arriveTangent=-1.5995, interpMode=RCIM_Cubic, leaveTangent=-1.5995, time=0.4653, value=0.5907}
          <at 0.73 0.0589 RCIM_Cubic>
         ]}
        YCurve {keys [
          <at 0.1712 1.0363>
          {arriveTangent=-1.5241, interpMode=RCIM_Cubic, leaveTangent=-1.5241, time=0.3943, value=0.6055}
          {arriveTangent=-1.6299, interpMode=RCIM_Cubic, leaveTangent=-1.6299, time=0.4653, value=0.5881}
          <at 0.73 0.0583 RCIM_Cubic>
         ]}
        ZCurve {keys [
          <at 0.1712 0.4143>
          {arriveTangent=0.6727, interpMode=RCIM_Cubic, leaveTangent=0.6727, time=0.3943, value=0.6173}
          {arriveTangent=-1.6543, interpMode=RCIM_Cubic, leaveTangent=-1.6543, time=0.4653, value=0.6121}
          <at 0.73 0.062 RCIM_Cubic>
         ]}
       }})
     "Float (Alpha)"=&ClampFloat(Float=&FloatFromCurve(FloatCurve=DataInterface() {propertyValues {Curve {keys [<at 0.1016 1>, {arriveTangent=-1.1131, interpMode=RCIM_Cubic, leaveTangent=-1.1131, time=0.4637, value=0.2405}, <at 1 0>]}}}))
    ))
   &AccelerationForce(Acceleration=$Particles.Acceleration)
   &SolveForcesAndVelocity(Clamp Velocity=false, Limit Acceleration=false)
  ]
  rendererClasses [&NiagaraRibbonRendererProperties]
  renderers [{rendererClass=&NiagaraRibbonRendererProperties}]
 }
 fuse {
  bEnabled=false
  emitterUpdateScript [
   &EmitterState(Inactive Response=::NewEnumerator1, Life Cycle Mode=::NewEnumerator1, Loop Duration=1, UseLoopDelay=true)
   &SpawnBurst_Instantaneous(Use Spawn Probability=false, UseLoopCountLimit=false)
   &SpawnRate(SpawnRate=10, Use Spawn Probability=false)
  ]
  particleSpawnScript [
   /SetVariables(Particles.SpriteRotationRate=&RandomRangeFloat(Minimum=0, Maximum=0.01)) {isSetParameters=true}
   &InitializeParticle(
    Lifetime=&RandomRangeFloat(Minimum=0.1, Maximum=0.2)
    Sprite Rotation Angle=&RandomRangeFloat(Minimum=0, Maximum=1)
    Sprite Rotation Mode=::NewEnumerator1
    Sprite Size=&RandomRangeVector2D(Minimum=<30 5>, Maximum=<200 35>)
    Sprite Size Mode=::NewEnumerator4
    UsePositionOffset=false
    Write Material Random=true
   )
   &CameraOffset(Camera Offset Amount=&RandomRangeFloat(Minimum=10, Maximum=60))
  ]
  particleUpdateScript [
   &ParticleState()
   &SpriteRotationRate(Rotation Rate=&AngleConversion(
     Angle:NiagaraFloat=$Particles.SpriteRotationRate
     Angle Input:ENiagara_AngleInput=::NewEnumerator1
     Angle Output:ENiagara_AngleInput=::NewEnumerator0
    ))
   &ScaleSpriteSize(
    Scale Factor=&Vector2DFromFloat(Value=&FloatFromCurve(FloatCurve=DataInterface() {propertyValues {Curve {keys [
          <at -0.0011 0.0573>
          {arriveTangent=-0.0322, interpMode=RCIM_Cubic, leaveTangent=-0.0322, time=0.0867, value=0.7609}
          {interpMode=RCIM_Cubic, leaveTangent=0.0935, time=1.0043, value=0.0249}
         ]}}}))
    Scale Sprite Size Mode=::NewEnumerator2
   )
   &Color(Color=&MakeLinearColorFromVectorAndFloat(
     "Vector (RGB)"=<238.589 81.2308 24.9103>
     "Float (Alpha)"=&FloatFromCurve(FloatCurve=DataInterface() {propertyValues {Curve {keys [<at 0 0 RCIM_Cubic>,<at 0.0967 0.1386 RCIM_Cubic>,<at 1 0>]}}})
    ))
  ]
  rendererClasses [&NiagaraSpriteRendererProperties]
  renderers [{rendererClass=&NiagaraSpriteRendererProperties}]
 }
}
summary {
 systemName=NS_Grenade_Fuse_Converted
 emitters [
  {bEnabled=false, emitterName=fuse, rendererClasses [&NiagaraSpriteRendererProperties]}
  {emitterName=GPU_up_Delay, simTarget=GPUComputeSim, rendererClasses [&NiagaraSpriteRendererProperties]}
  {emitterName=Ribbon, rendererClasses [&NiagaraRibbonRendererProperties]}
  {
   emitterName=Particle_Emitter
   rendererClasses [&NiagaraSpriteRendererProperties, &NiagaraLightRendererProperties]
  }
  {emitterName=Fuse001, rendererClasses [&NiagaraSpriteRendererProperties]}
 ]
 userVariables []
}
